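# Physics asset for the Neon Graffiti backpack cosmetic, defining bodies and joint limits for its dangling cans and flap.
# Capsules cover the left and right can bones, a box covers the flap, and a tiny capsule on the pack bone serves as their shared anchor.
# Each of the three parts is constrained directly to the pack with narrow, asymmetric swing and twist limits, so any motion stays small.
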PhysicsAsset /BRCosmetics/Accessories/FORT_Backpacks/F_MED_Neon_Graffiti_Backpack/Meshes/F_MED_Neon_Graffiti_Pack_Backpack_Physics
bodies {
 dyn_can_l [{shapeType=Capsule, length=10.2295, radius=3.75, center=<-2.1984 -0.5879 3.2631>, rotation=<rot 90 0 0>}]
 dyn_can_r [{shapeType=Capsule, length=10.4718, radius=3.75, center=<2.1705 0.98 -3.3025>, rotation=<rot 90 0 -179.9999>}]
 dyn_flap [{shapeType=Box, extentX=21.2999, extentY=12.4623, extentZ=16.1399, center=<13.3576 -0.2851 -0.0202>}]
 pack [{shapeType=Capsule, length=1, radius=1}]
}
constraints [
 {bone1Name=dyn_can_l, bone2Name=pack, swing1LimitDegrees=10, swing2LimitDegrees=25, twistLimitDegrees=10}
 {bone1Name=dyn_flap, bone2Name=pack, swing1LimitDegrees=30, swing2LimitDegrees=5, twistLimitDegrees=5}
 {bone1Name=dyn_can_r, bone2Name=pack, swing1LimitDegrees=7, swing2LimitDegrees=5, twistLimitDegrees=10}
]
