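# Animation Blueprint for the Neon Graffiti backpack cosmetic, adding a vibrating speaker and secondary physics motion.
# Each update, speakerReverb_FNC layers Perlin noise and sine over game time to produce translation and scale values.
# These additively drive the dyn_speaker bone, then a RigidBody node simulates the backpack physics asset with overridden gravity.
# An Anim Dynamics node for dyn_flap sits in the graph, unconnected to the output pose.
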AnimBlueprint /BRCosmetics/Accessories/FORT_Backpacks/F_MED_Neon_Graffiti_Backpack/Meshes/F_MED_Neon_Graffiti_Pack_Backpack_AnimBP
target_skeleton=&/F_MED_Neon_Graffiti_Pack_Backpack_Skeleton
graphs [
 AnimGraph {nodes [
   Anim Dynamics - Bone: dyn_flap(Alpha=1, ExternalForce="0,0,0") {
    angularDampingOverride=1
    angularSpringConstant=1000
    bAngularSpring=true
    bGravityOverrideInSimSpace=true
    bOverrideAngularDamping=true
    gravityScale=0.35
    simulationSpace=RootRelative
    alphaBoolBlend {blendOption=Linear}
    boundBone {boneName=dyn_flap}
    componentLinearAccScale=<0.3 0.3 0.3>
    componentLinearVelScale=<0.3 0.3 0.3>
    physicsBodyDefinitions [{
      sphereCollisionRadius=0
      boundBone {boneName=dyn_flap}
      boxExtents=<10 20 20>
      constraintSetup {angularLimitsMax=<0 0 0>, angularLimitsMin=<0 0 -30>, angularTarget=<0 0 0>}
      localJointOffset=<10 0 0>
     }]
   }
   AnimGraphNode_SubInput_0 = Input Pose()
   n1 = "Transform (Modify) Bone - Bone: dyn_speaker"(
    ComponentPose=Local To Component(LocalPose=@AnimGraphNode_SubInput_0)
    Alpha=1
    Translation=Get reverbTran() -> (reverbTran:Vector)
    Rotation="0,0,0"
    Scale=Get reverbScale() -> (reverbScale:Vector)
   ) {
    rotationMode=BMM_Additive
    rotationSpace=BCS_BoneSpace
    scaleMode=BMM_Additive
    scaleSpace=BCS_BoneSpace
    translationMode=BMM_Additive
    translationSpace=BCS_BoneSpace
    alphaBoolBlend {blendOption=Linear}
    alphaScaleBiasClamp {inRange {max=25, min=10}}
    boneToModify {boneName=dyn_speaker}
   }
   n2 = RigidBody(ComponentPose=@n1) {
    bClampLinearTranslationLimitToRefPose=true
    bOverrideWorldGravity=true
    overridePhysicsAsset=&/F_MED_Neon_Graffiti_Pack_Backpack_Physics
    simulationSpace=BaseBoneSpace
    alphaBoolBlend {blendOption=Linear}
    baseBoneRef {boneName=root}
    componentLinearAccScale=<0.5 0.5 0.5>
    componentLinearVelScale=<0.5 0.5 0.5>
    overrideWorldGravity {z=-980}
   }
   Output Pose(Result=Component To Local(ComponentPose=@n2)) {layerGroup=None}
  ]}
 EventGraph {nodes [
   Event BlueprintUpdateAnimation()
   > n1 = speakerReverb_FNC!(execute, self, frequency:double=5, amp:double=0.3) -> (then, tran_output:double, scale_output:double)
   > Set reverbTran(reverbTran:Vector=MakeVector(Y=@n1)) -> (Output_Get:Vector)
   > Set reverbScale(reverbScale:Vector=MakeVector(X=@n1.scale_output, Y=@n1.scale_output, Z=@n1.scale_output)) -> (Output_Get:Vector)
   TryGetPawnOwner()
  ]}
 speakerReverb_FNC {nodes [
   n1 = speakerReverb_FNC!() -> (then, frequency:double=~, amp:double=~) {is=title K2Node_FunctionEntry}
   n2 = GetGameTimeInSeconds()
   n3 = "Sin (Radians)"(A=PerlinNoise1D!(Value:float=float * float(A=@n1, B=@n2)) -> (ReturnValue:float))
   > Return Node!(
    execute
    tran_output:double=PerlinNoise1D(Value=float * float(A=@n3, B=@n1.amp))
    scale_output:double=Make Perlin Noise and Remap!(Value:float=@n3, RangeOutMin:float=0.9, RangeOutMax:float=1.1) -> (ReturnValue:float)
   )
  ]}
]
variables [
 reverbScale {type=Vector}
 reverbTran {type=Vector}
]
